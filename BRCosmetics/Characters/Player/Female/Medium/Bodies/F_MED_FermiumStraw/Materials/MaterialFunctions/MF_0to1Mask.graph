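# Material function in the F_MED_FermiumStraw cosmetic's folder that builds a square mask over the 0-1 UV range.
# U and V each feed a scalar SphereMask centered at 0.5 plus Padding with a 0.5 radius, acting as a one-dimensional range test.
# Multiplying both results gives the square.
# Padding shifts the square along both axes without resizing it, and UV defaults to the texture coordinate.
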MaterialFunction /BRCosmetics/Characters/Player/Female/Medium/Bodies/F_MED_FermiumStraw/Materials/MaterialFunctions/MF_0to1Mask
graph {nodes [
  n1 = Add(A=Constant() {r=0.5}, B=FunctionInput() {bUsePreviewValueAsDefault=true, inputName=Padding, inputType=FunctionInput_Scalar, sortPriority=10})
  n2 = MaterialFunctionCall(Float2=FunctionInput(Preview=TextureCoordinate()) {bUsePreviewValueAsDefault=true, inputName=UV, inputType=FunctionInput_Vector2}) {materialFunction=&BreakOutFloat2Components}
  FunctionOutput(None=Multiply(A=SphereMask(A=@n1, B=@n2.R) {attenuationRadius=0.5}, B=SphereMask(A=@n2.G, B=@n1) {attenuationRadius=0.5}))
 ]}
