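# Material function in the FermiumStraw character body's materials that remaps input UVs into normalized space for a UV island given by min and max bounds.
# It outputs a stretched island UV and a 1:1 aspect version, each produced by a subtract and divide, the latter dividing by the larger dimension.
# An If builds centered UVs by offsetting one axis by half a difference.
# Each output gets a 0-1 range mask from multiplied per-axis SphereMasks centered at 0.5.
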MaterialFunction /BRCosmetics/Characters/Player/Female/Medium/Bodies/F_MED_FermiumStraw/Materials/MaterialFunctions/MF_NormalizeUVIsland
graph {nodes [
  n1 = Constant() {r=0.5}
  n2 = Subtract(A=NamedRerouteUsage(), B=NamedRerouteUsage())
  n3 = NamedRerouteUsage()
  n4 = MaterialFunctionCall(Float2=FunctionOutput(None=Divide(A=@n2, B=@n3)) {outputName="Island UV (V2)"}) {materialFunction=&BreakOutFloat2Components}
  FunctionOutput(None=Multiply(A=SphereMask(A=@n1, B=@n4.R) {attenuationRadius=0.5}, B=SphereMask(A=@n4.G, B=@n1) {attenuationRadius=0.5})) {outputName="Island 0-1 Mask (S)", sortPriority=5}
  n5 = Constant() {r=0.5}
  n6 = Subtract(A=NamedRerouteUsage(), B=NamedRerouteUsage())
  n7 = Max(A=NamedRerouteUsage(), B=NamedRerouteUsage())
  n8 = MaterialFunctionCall(Float2=FunctionOutput(None=Divide(A=@n6, B=@n7)) {outputName="Island UV - 1:1 Aspect Ratio (V2)", sortPriority=10}) {materialFunction=&BreakOutFloat2Components}
  FunctionOutput(None=Multiply(A=SphereMask(A=@n5, B=@n8.R) {attenuationRadius=0.5}, B=SphereMask(A=@n8.G, B=@n5) {attenuationRadius=0.5})) {outputName="Island UV - 1:1 Aspect Ratio 0-1 Mask (S)", sortPriority=15}
  NamedRerouteDeclaration(None=FunctionInput() {bUsePreviewValueAsDefault=true, inputName=Island UV Min, inputType=FunctionInput_Vector2, sortPriority=10}) {name="Island UV Min (V2)", nodeColor=<rgba 0 0.2734 1 1>}
  n9 = FunctionInput() {bUsePreviewValueAsDefault=true, inputName=Island UV Max, inputType=FunctionInput_Vector2, sortPriority=20, previewValue=<1 1 0 1>}
  NamedRerouteDeclaration(None=@n9) {name="Island UV Max (V2)", nodeColor=<rgba 0 1 0.3478 1>}
  n10 = NamedRerouteUsage()
  n11 = NamedRerouteUsage()
  n12 = NamedRerouteUsage()
  n13 = Add(A=Multiply(A=Subtract(A=@n11, B=@n12)) {constB=0.5}, B=NamedRerouteUsage())
  n14 = NamedRerouteUsage()
  n15 = NamedRerouteUsage()
  n16 = Add(A=NamedRerouteUsage(), B=Multiply(A=Subtract(A=@n14, B=@n15)) {constB=0.5})
  n17 = NamedRerouteUsage()
  NamedRerouteDeclaration(None=If(A=NamedRerouteUsage(), B=NamedRerouteUsage(), "A > B"=AppendVector(A=@n10, B=@n13), "A < B"=AppendVector(A=@n16, B=@n17))) {name="Centered UV (V2)", nodeColor=<rgba 0 1 0.7656 1>}
  MaterialExpressionNamedRerouteDeclaration_9 = NamedRerouteDeclaration(None=Subtract(A=NamedRerouteUsage(), B=NamedRerouteUsage())) {name="Island UV Size (V2)", nodeColor=<rgba 0 0.6953 1 1>}
  n18 = Convert(None=@MaterialExpressionNamedRerouteDeclaration_9) {
   nodeName=Break Float2
   convertInputs [{
     type=Vector2
     expressionInput {
      expression=&self:MaterialExpressionNamedRerouteDeclaration_9
      inputName=None
      mask=0
      maskA=0
      maskB=0
      maskG=0
      maskR=0
      outputIndex=0
     }
    }]
   convertOutputs [{type=Scalar}, {type=Scalar}]
  }
  NamedRerouteDeclaration(None=@n18) {name="Island U Size (S)", nodeColor=<rgba 0.0078 0 1 1>}
  NamedRerouteDeclaration(None=@n18) {name="Island V Size (S)", nodeColor=<rgba 0.3125 1 0 1>}
  MaterialExpressionNamedRerouteDeclaration_0 = NamedRerouteDeclaration(None=FunctionInput(Preview=TextureCoordinate()) {bUsePreviewValueAsDefault=true, inputName=UV in, inputType=FunctionInput_Vector2}) {name="UV in (V2)", nodeColor=<rgba 1 0.2087 0 1>}
  n19 = Convert(None=@MaterialExpressionNamedRerouteDeclaration_0) {
   nodeName=Break Float2
   convertInputs [{
     type=Vector2
     expressionInput {
      expression=&self:MaterialExpressionNamedRerouteDeclaration_0
      inputName=None
      mask=0
      maskA=0
      maskB=0
      maskG=0
      maskR=0
      outputIndex=0
     }
    }]
   convertOutputs [{type=Scalar}, {type=Scalar}]
  }
  NamedRerouteDeclaration(None=@n19) {name="U in (S)", nodeColor=<rgba 0 0.5565 1 1>}
  NamedRerouteDeclaration(None=@n19) {name="V in (S)", nodeColor=<rgba 0.3125 1 0 1>}
 ]}
